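# Physics asset for the PirouetteWeld backpack cosmetic, providing bodies and joints for a pair of dynamic wings.
# Each wing is a capsule chain from wing_00 through dyn_wing_08, with twist locked and per-joint swing limit angles.
# Capsules on the spine, pelvis, arms and legs, joined by angularly free constraints, plus an unconstrained floorCOL box, supply character and floor collision shapes.
# No constraint links the wing roots to the body.
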PhysicsAsset /BRCosmetics/Accessories/FORT_Backpacks/Backpack_F_MED_PirouetteWeld/Meshes/F_MED_PirouetteWeld_Pack_Physics
bodies {
 calf_l [{shapeType=Capsule, length=31, radius=10.75, shapeName=calf_l_capsule, center=<-21.6338 0 0>, rotation=<rot 90 0 0>}]
 calf_r [{shapeType=Capsule, length=31, radius=11.5, shapeName=calf_r_capsule, center=<21.3497 0 0>, rotation=<rot 90 0 0>}]
 dyn_wing_02_l [{
   shapeType=Capsule
   length=5.2589
   radius=3.666
   shapeName=dyn_wing_02_l_capsule
   center=<5.6687 0.3766 1.2807>
   rotation=<rot 0 88.9345 -87.0141>
  }]
 dyn_wing_02_r [{
   shapeType=Capsule
   length=5.3085
   radius=3.7056
   shapeName=dyn_wing_02_r_capsule
   center=<5.6606 0.1518 -1.4482>
   rotation=<rot 0 -90.7988 -85.332>
  }]
 dyn_wing_03_l [{
   shapeType=Capsule
   length=8.3901
   radius=6.8207
   shapeName=dyn_wing_03_l_capsule
   center=<8.9231 0.612 1.1838>
   rotation=<rot 14.1102 -90 -128.7953>
  }]
 dyn_wing_03_r [{
   shapeType=Capsule
   length=8.4052
   radius=6.8357
   shapeName=dyn_wing_03_r_capsule
   center=<9.0079 0.5266 -1.6205>
   rotation=<rot -9.7747 -90 -46.9308>
  }]
 dyn_wing_04_l [{
   shapeType=Capsule
   length=11.4688
   radius=10.4664
   shapeName=dyn_wing_04_l_capsule
   center=<9.0793 -0.2976 2.5883>
   rotation=<rot 11.8391 -90 21.9876>
  }]
 dyn_wing_04_r [{
   shapeType=Capsule
   length=11.4632
   radius=11.227
   shapeName=dyn_wing_04_r_capsule
   center=<8.3579 -0.5231 -3.2833>
   rotation=<rot -7.2131 -90 -21.357>
  }]
 dyn_wing_05_l [{
   shapeType=Capsule
   length=11.0402
   radius=10.6089
   shapeName=dyn_wing_05_l_capsule
   center=<10.1192 -0.7926 -0.2947>
   rotation=<rot 8.1757 -90 8.1221>
  }]
 dyn_wing_05_r [{
   shapeType=Capsule
   length=11.0705
   radius=9.99
   shapeName=dyn_wing_05_r_capsule
   center=<9.4229 -0.7733 -0.3003>
   rotation=<rot -3.7039 -90 -8.2955>
  }]
 dyn_wing_06_l [{
   shapeType=Capsule
   length=8.1517
   radius=5.1271
   shapeName=dyn_wing_06_l_capsule
   center=<6.8959 -1.0813 -0.1233>
   rotation=<rot 6.1045 -90 2.1754>
  }]
 dyn_wing_06_r [{
   shapeType=Capsule
   length=8.1591
   radius=5.1077
   shapeName=dyn_wing_06_r_capsule
   center=<6.8867 -1.1117 -0.4695>
   rotation=<rot -1.7341 -90 -3.1828>
  }]
 dyn_wing_07_l [{
   shapeType=Capsule
   length=6.2811
   radius=5.9288
   shapeName=dyn_wing_07_l_capsule
   center=<3.2033 -0.9817 1.2128>
   rotation=<rot 5.5279 -90 -102.5204>
  }]
 dyn_wing_07_r [{
   shapeType=Capsule
   length=6.1572
   radius=5.8747
   shapeName=dyn_wing_07_r_capsule
   center=<3.3143 -1.1924 -0.7219>
   rotation=<rot -1.6084 -90 -92.0995>
  }]
 dyn_wing_08_l [{
   shapeType=Capsule
   length=2.3921
   radius=2.0021
   shapeName=dyn_wing_08_l_capsule
   center=<1.1915 -0.8822 -0.3689>
   rotation=<rot 0 -92.0568 -88.7062>
  }]
 dyn_wing_08_r [{
   shapeType=Capsule
   length=2.3921
   radius=1.9975
   shapeName=dyn_wing_08_r_capsule
   center=<1.1694 -0.9362 0.0605>
   rotation=<rot 0 88.0174 -88.6167>
  }]
 floorCOL [{shapeType=Box, extentX=150, extentY=200, extentZ=1, shapeName=floorCOL_box}]
 lowerarm_l [{shapeType=Capsule, length=24.1997, radius=5.6176, shapeName=lowerarm_l_capsule, center=<14.9037 0 0>, rotation=<rot 90 0 0>}]
 lowerarm_r [{shapeType=Capsule, length=26.8581, radius=5.5686, shapeName=lowerarm_r_capsule, center=<-17.8835 -2.8944 0>, rotation=<rot 90 0 0>}]
 pelvis [{shapeType=Capsule, length=21.178, radius=9, shapeName=pelvis_capsule, center=<-9.9247 -3.7916 0>, rotation=<rot 0 -0.0001 0>}]
 spine_01 [{shapeType=Capsule, length=23.3193, radius=10.002, shapeName=spine_01_capsule, center=<0.0013 2.4144 -0.5593>, rotation=<rot 0.1281 0 0>}]
 spine_03 [{
   shapeType=Capsule
   length=25.8666
   radius=7.0836
   shapeName=spine_03_capsule
   center=<3.6771 2.2146 0.1099>
   rotation=<rot 1.8255 6.0342 0.3014>
  }]
 spine_04 [{shapeType=Capsule, length=4.6222, radius=9.0201, shapeName=spine_04_capsule, center=<10.737 4.3078 0>, rotation=<rot 90 -179.1048 173.9881>}]
 thigh_l [{shapeType=Capsule, length=25.922, radius=7.5, shapeName=thigh_l_capsule, center=<-17.4423 -1.6724 0>, rotation=<rot 90 180 -175>}]
 thigh_r [{shapeType=Capsule, length=25.243, radius=9.5, shapeName=thigh_r_capsule, center=<18.8343 2.0229 0>, rotation=<rot 90 179.5524 -175.4476>}]
 upperarm_l [{shapeType=Capsule, length=23.2581, radius=5.4182, shapeName=upperarm_l_capsule, center=<13.102 0 0>, rotation=<rot 90 0 0>}]
 upperarm_r [{
   shapeType=Capsule
   length=25.5491
   radius=5.4397
   shapeName=upperarm_r_capsule
   center=<-13.5448 -2.0853 0>
   rotation=<rot 90 -179.2838 -176.8839>
  }]
 wing_00_l [{
   shapeType=Capsule
   length=0.505
   radius=0.505
   shapeName=wing_00_l_capsule
   center=<6.0323 1.8289 -0.3717>
   rotation=<rot 0 -79.6509 -87.6035>
  }]
 wing_00_r [{shapeType=Capsule, length=0.505, radius=0.505, shapeName=wing_00_r_capsule, center=<6.0346 1.8923 0.183>, rotation=<rot 0 100.5235 -89.0832>}]
 wing_01_l [{
   shapeType=Capsule
   length=9.1757
   radius=3.9549
   shapeName=wing_01_l_capsule
   center=<-0.9917 -1.9172 0.1777>
   rotation=<rot 0 95.4078 -87.4059>
  }]
 wing_01_r [{
   shapeType=Capsule
   length=9.9077
   radius=3.9952
   shapeName=wing_01_r_capsule
   center=<-0.2567 -2.0169 -0.121>
   rotation=<rot 0 -84.6649 -86.2418>
  }]
}
constraints [
 {bone1Name=dyn_wing_05_r, bone2Name=dyn_wing_04_r, swing1LimitDegrees=30, swing2LimitDegrees=20, twistLimitDegrees=45, twistMotion=Locked}
 {bone1Name=dyn_wing_07_r, bone2Name=dyn_wing_06_r, swing1LimitDegrees=30, swing2LimitDegrees=6, twistLimitDegrees=45, twistMotion=Locked}
 {bone1Name=dyn_wing_05_l, bone2Name=dyn_wing_04_l, swing1LimitDegrees=30, swing2LimitDegrees=20, twistLimitDegrees=45, twistMotion=Locked}
 {bone1Name=dyn_wing_07_l, bone2Name=dyn_wing_06_l, swing1LimitDegrees=30, swing2LimitDegrees=6, twistLimitDegrees=45, twistMotion=Locked}
 {
  bone1Name=spine_01
  bone2Name=pelvis
  swing1LimitDegrees=45
  swing1Motion=Free
  swing2LimitDegrees=45
  swing2Motion=Free
  twistLimitDegrees=45
  twistMotion=Free
 }
 {
  bone1Name=lowerarm_l
  bone2Name=upperarm_l
  swing1LimitDegrees=45
  swing1Motion=Free
  swing2LimitDegrees=45
  swing2Motion=Free
  twistLimitDegrees=45
  twistMotion=Free
 }
 {
  bone1Name=spine_03
  bone2Name=spine_01
  swing1LimitDegrees=45
  swing1Motion=Free
  swing2LimitDegrees=45
  swing2Motion=Free
  twistLimitDegrees=45
  twistMotion=Free
 }
 {
  bone1Name=spine_04
  bone2Name=spine_03
  swing1LimitDegrees=45
  swing1Motion=Free
  swing2LimitDegrees=45
  swing2Motion=Free
  twistLimitDegrees=45
  twistMotion=Free
 }
 {
  bone1Name=upperarm_l
  bone2Name=spine_04
  swing1LimitDegrees=45
  swing1Motion=Free
  swing2LimitDegrees=45
  swing2Motion=Free
  twistLimitDegrees=45
  twistMotion=Free
 }
 {
  bone1Name=upperarm_r
  bone2Name=spine_04
  swing1LimitDegrees=45
  swing1Motion=Free
  swing2LimitDegrees=45
  swing2Motion=Free
  twistLimitDegrees=45
  twistMotion=Free
 }
 {
  bone1Name=lowerarm_r
  bone2Name=upperarm_r
  swing1LimitDegrees=45
  swing1Motion=Free
  swing2LimitDegrees=45
  swing2Motion=Free
  twistLimitDegrees=45
  twistMotion=Free
 }
 {
  bone1Name=thigh_l
  bone2Name=pelvis
  swing1LimitDegrees=45
  swing1Motion=Free
  swing2LimitDegrees=45
  swing2Motion=Free
  twistLimitDegrees=45
  twistMotion=Free
 }
 {
  bone1Name=calf_l
  bone2Name=thigh_l
  swing1LimitDegrees=45
  swing1Motion=Free
  swing2LimitDegrees=45
  swing2Motion=Free
  twistLimitDegrees=45
  twistMotion=Free
 }
 {
  bone1Name=thigh_r
  bone2Name=pelvis
  swing1LimitDegrees=45
  swing1Motion=Free
  swing2LimitDegrees=45
  swing2Motion=Free
  twistLimitDegrees=45
  twistMotion=Free
 }
 {
  bone1Name=calf_r
  bone2Name=thigh_r
  swing1LimitDegrees=45
  swing1Motion=Free
  swing2LimitDegrees=45
  swing2Motion=Free
  twistLimitDegrees=45
  twistMotion=Free
 }
 {bone1Name=dyn_wing_02_l, bone2Name=wing_01_l, swing1LimitDegrees=13, swing2LimitDegrees=5, twistLimitDegrees=45, twistMotion=Locked}
 {bone1Name=dyn_wing_03_l, bone2Name=dyn_wing_02_l, swing1LimitDegrees=30, swing2LimitDegrees=20, twistLimitDegrees=45, twistMotion=Locked}
 {bone1Name=dyn_wing_04_l, bone2Name=dyn_wing_03_l, swing1LimitDegrees=30, swing2LimitDegrees=22, twistLimitDegrees=45, twistMotion=Locked}
 {bone1Name=dyn_wing_06_l, bone2Name=dyn_wing_05_l, swing1LimitDegrees=30, swing2LimitDegrees=7, twistLimitDegrees=45, twistMotion=Locked}
 {bone1Name=dyn_wing_08_l, bone2Name=dyn_wing_07_l, swing1LimitDegrees=30, swing2LimitDegrees=5, twistLimitDegrees=45, twistMotion=Locked}
 {bone1Name=dyn_wing_02_r, bone2Name=wing_01_r, swing1LimitDegrees=13, swing2LimitDegrees=5, twistLimitDegrees=45, twistMotion=Locked}
 {bone1Name=dyn_wing_03_r, bone2Name=dyn_wing_02_r, swing1LimitDegrees=30, swing2LimitDegrees=20, twistLimitDegrees=45, twistMotion=Locked}
 {bone1Name=dyn_wing_04_r, bone2Name=dyn_wing_03_r, swing1LimitDegrees=30, swing2LimitDegrees=22, twistLimitDegrees=45, twistMotion=Locked}
 {bone1Name=dyn_wing_06_r, bone2Name=dyn_wing_05_r, swing1LimitDegrees=30, swing2LimitDegrees=7, twistLimitDegrees=45, twistMotion=Locked}
 {bone1Name=dyn_wing_08_r, bone2Name=dyn_wing_07_r, swing1LimitDegrees=30, swing2LimitDegrees=5, twistLimitDegrees=45, twistMotion=Locked}
 {bone1Name=wing_01_l, bone2Name=wing_00_l, swing1LimitDegrees=13, swing2LimitDegrees=5, twistLimitDegrees=45, twistMotion=Locked}
 {bone1Name=wing_01_r, bone2Name=wing_00_r, swing1LimitDegrees=13, swing2LimitDegrees=5, twistLimitDegrees=45, twistMotion=Locked}
]
