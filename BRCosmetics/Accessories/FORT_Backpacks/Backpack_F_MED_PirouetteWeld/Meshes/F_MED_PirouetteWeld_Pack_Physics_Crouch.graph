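# Physics asset for the crouch variant of the PirouetteWeld backpack, simulating its two wings.
# Each wing chains capsule bodies from Wing_01 through dyn_wing_08, with multiple capsules widening the middle segments.
# Wing joints mostly lock twist and favor swing1, with swing2 much tighter on the right than the left.
# Torso, arm and leg capsules with free constraints, plus a thin floorCOL box, provide body and ground collision shapes.
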PhysicsAsset /BRCosmetics/Accessories/FORT_Backpacks/Backpack_F_MED_PirouetteWeld/Meshes/F_MED_PirouetteWeld_Pack_Physics_Crouch
bodies {
 Wing_01_L [{
   shapeType=Capsule
   length=8.8079
   radius=4.8023
   shapeName=Wing_01_L_capsule
   center=<-3.9224 0.5632 0.2571>
   rotation=<rot 0.1978 -96.5188 -86.847>
  }]
 Wing_01_R [{
   shapeType=Capsule
   length=10.0579
   radius=4.5523
   shapeName=Wing_01_R_capsule
   center=<-4.1052 0.6272 -0.0972>
   rotation=<rot -0.0554 -91.9279 -91.2454>
  }]
 calf_l [{shapeType=Capsule, length=31, radius=10.75, shapeName=calf_l_capsule, center=<-21.6338 0 0>, rotation=<rot 90 0 0>}]
 calf_r [{shapeType=Capsule, length=31, radius=10.75, shapeName=calf_r_capsule, center=<21.3497 0 0>, rotation=<rot 90 0 0>}]
 dyn_wing_02_l [{
   shapeType=Capsule
   length=9.2642
   radius=3.973
   shapeName=Wing_02_R_capsule
   center=<-0.0072 -0.4399 -0.0214>
   rotation=<rot 89.8735 -179.9985 176.4027>
  }]
 dyn_wing_02_r [{shapeType=Capsule, length=8.221, radius=4.56, shapeName=Wing_02_R_capsule, center=<0.0862 0 -0.0047>, rotation=<rot 90 106.2566 106.2565>}]
 dyn_wing_03_l [
  {shapeType=Capsule, length=13.25, radius=5, shapeName=Wing_03_R_capsule, center=<1.4601 0 0>, rotation=<rot 90 0 0>}
  {shapeType=Capsule, length=13.75, radius=5, shapeName=Wing_03_L_capsule, center=<5.3557 0 -3.1992>, rotation=<rot 80 0 0>}
  {shapeType=Capsule, length=15, radius=3.8572, shapeName=Wing_03_L_capsule, center=<6.8663 0 4.0549>, rotation=<rot 83.1368 -180 -180>}
 ]
 dyn_wing_03_r [
  {shapeType=Capsule, length=13.25, radius=5, shapeName=Wing_03_R_capsule, center=<0 1.4601 0>, rotation=<rot 90 0 0>}
  {shapeType=Capsule, length=15, radius=5, shapeName=Wing_03_R_capsule, center=<6.4693 0 3.7755>, rotation=<rot 80.9602 -180 180>}
  {shapeType=Capsule, length=15, radius=4, shapeName=Wing_03_R_capsule, center=<6.9786 0 -4.0374>, rotation=<rot 81.1938 0.0001 0>}
 ]
 dyn_wing_04_l [
  {shapeType=Capsule, length=13, radius=5, shapeName=Wing_04_R_capsule, center=<8.784 0 0.1124>, rotation=<rot 90 9 9>}
  {shapeType=Capsule, length=13, radius=5, shapeName=Wing_04_L_capsule, center=<7.7079 0 -5.7531>, rotation=<rot 81.8991 0 0>}
  {shapeType=Capsule, length=15, radius=3.8572, shapeName=Wing_04_L_capsule, center=<8.7212 0 5.9533>, rotation=<rot 84.4982 -179.9999 -180>}
 ]
 dyn_wing_04_r [
  {shapeType=Capsule, length=11.5, radius=5, shapeName=Wing_04_R_capsule, center=<8.0082 0 0.277>, rotation=<rot 90 0 0>}
  {shapeType=Capsule, length=11.5, radius=5, shapeName=Wing_04_R_capsule, center=<8.2928 0 5.9937>, rotation=<rot 82.6337 -180 180>}
  {shapeType=Capsule, length=15, radius=3.8572, shapeName=Wing_04_R_capsule, center=<8.609 0 -6.4134>, rotation=<rot 84.4506 0.0002 0.0001>}
 ]
 dyn_wing_05_l [
  {shapeType=Capsule, length=13, radius=5, shapeName=Wing_05_R_capsule, center=<8.9462 -0.0211 -0.0476>, rotation=<rot 90 0 0>}
  {shapeType=Capsule, length=13.75, radius=5, shapeName=Wing_05_L_capsule, center=<8.254 0 -6.7428>, rotation=<rot 88.0619 -179.9999 -179.9999>}
 ]
 dyn_wing_05_r [
  {shapeType=Capsule, length=10.5, radius=5, shapeName=Wing_05_R_capsule, center=<8.2481 0.0254 -0.0445>, rotation=<rot 90 0 0>}
  {shapeType=Capsule, length=10.5, radius=5, shapeName=Wing_05_R_capsule, center=<7.1843 0 6.7275>, rotation=<rot 88.1769 0.0001 0.0001>}
 ]
 dyn_wing_06_l [
  {shapeType=Capsule, length=12, radius=4.75, shapeName=Wing_06_R_capsule, center=<5.5029 0 0.1348>, rotation=<rot 89.4037 -0.0003 -0.0003>}
  {shapeType=Capsule, length=15, radius=5, shapeName=Wing_06_L_capsule, center=<6.6831 0 -4.4983>, rotation=<rot 77.0274 -180 -180>}
 ]
 dyn_wing_06_r [
  {shapeType=Capsule, length=12, radius=5, shapeName=Wing_06_R_capsule, center=<6.8352 0 0.2138>, rotation=<rot 90 0 0>}
  {shapeType=Capsule, length=16.25, radius=5, shapeName=Wing_06_R_capsule, center=<6.1086 0 4.4532>, rotation=<rot 77.5921 0.0001 0>}
 ]
 dyn_wing_07_l [{shapeType=Capsule, length=6, radius=5, shapeName=Wing_07_R_capsule, center=<4.5381 0 -0.0998>, rotation=<rot 88.3494 0 -0.0001>}]
 dyn_wing_07_r [{shapeType=Capsule, length=6.5, radius=5, shapeName=Wing_07_R_capsule, center=<4.9551 0 0.1672>, rotation=<rot 87.0817 180 179.9999>}]
 dyn_wing_08_l [{
   shapeType=Capsule
   length=6.5
   radius=4.25
   shapeName=Wing_08_R_capsule
   center=<3.4813 0.0007 0.0297>
   rotation=<rot 89.442 -0.0001 -0.0001>
  }]
 dyn_wing_08_r [{shapeType=Capsule, length=6.5, radius=4.25, shapeName=Wing_08_R_capsule, center=<3.0848 -0.0002 0.2095>, rotation=<rot 90 0 0>}]
 floorCOL [{shapeType=Box, extentX=150, extentY=200, extentZ=1, shapeName=floorCOL_box}]
 lowerarm_l [{shapeType=Capsule, length=20.8824, radius=5.6176, shapeName=lowerarm_l_capsule, center=<15.8105 0 0>, rotation=<rot 90 0 0>}]
 lowerarm_r [{shapeType=Capsule, length=20.9314, radius=5.5686, shapeName=lowerarm_r_capsule, center=<-19.217 -2.8944 0>, rotation=<rot 90 0 0>}]
 pelvis [{shapeType=Capsule, length=23.879, radius=21.1958, shapeName=pelvis_capsule, center=<-14.4183 -2.7141 0>, rotation=<rot 0 -0.0001 0>}]
 spine_01 [{shapeType=Capsule, length=2.75, radius=10.75, shapeName=spine_01_capsule, rotation=<rot 90 0 0>}]
 spine_03 [{shapeType=Capsule, length=2.75, radius=10.75, shapeName=spine_03_capsule, rotation=<rot 90 0 0>}]
 spine_04 [{shapeType=Capsule, length=4.6222, radius=9.0201, shapeName=spine_04_capsule, center=<10.737 0 0>, rotation=<rot 90 0 0>}]
 thigh_l [{shapeType=Capsule, length=25.922, radius=15, shapeName=thigh_l_capsule, center=<-17.5879 -3.3362 0>, rotation=<rot 90 180 -175>}]
 thigh_r [{shapeType=Capsule, length=25.243, radius=15, shapeName=thigh_r_capsule, center=<19.1432 5.5531 0>, rotation=<rot 90 179.5524 -175.4476>}]
 upperarm_l [{shapeType=Capsule, length=17.5818, radius=5.4182, shapeName=upperarm_l_capsule, center=<10.7784 0 0>, rotation=<rot 90 0 0>}]
 upperarm_r [{shapeType=Capsule, length=18.0603, radius=5.4397, shapeName=upperarm_r_capsule, center=<-11.1254 -2.2637 0>, rotation=<rot 90 0 0>}]
}
constraints [
 {
  bone1Name=spine_01
  bone2Name=pelvis
  swing1LimitDegrees=45
  swing1Motion=Free
  swing2LimitDegrees=45
  swing2Motion=Free
  twistLimitDegrees=45
  twistMotion=Free
 }
 {
  bone1Name=spine_03
  bone2Name=spine_01
  swing1LimitDegrees=45
  swing1Motion=Free
  swing2LimitDegrees=45
  swing2Motion=Free
  twistLimitDegrees=45
  twistMotion=Free
 }
 {
  bone1Name=spine_04
  bone2Name=spine_03
  swing1LimitDegrees=45
  swing1Motion=Free
  swing2LimitDegrees=45
  swing2Motion=Free
  twistLimitDegrees=45
  twistMotion=Free
 }
 {
  bone1Name=lowerarm_l
  bone2Name=upperarm_l
  swing1LimitDegrees=45
  swing1Motion=Free
  swing2LimitDegrees=45
  swing2Motion=Free
  twistLimitDegrees=45
  twistMotion=Free
 }
 {
  bone1Name=upperarm_l
  bone2Name=spine_04
  swing1LimitDegrees=45
  swing1Motion=Free
  swing2LimitDegrees=45
  swing2Motion=Free
  twistLimitDegrees=45
  twistMotion=Free
 }
 {
  bone1Name=upperarm_r
  bone2Name=spine_04
  swing1LimitDegrees=45
  swing1Motion=Free
  swing2LimitDegrees=45
  swing2Motion=Free
  twistLimitDegrees=45
  twistMotion=Free
 }
 {
  bone1Name=lowerarm_r
  bone2Name=upperarm_r
  swing1LimitDegrees=45
  swing1Motion=Free
  swing2LimitDegrees=45
  swing2Motion=Free
  twistLimitDegrees=45
  twistMotion=Free
 }
 {
  bone1Name=thigh_l
  bone2Name=pelvis
  swing1LimitDegrees=45
  swing1Motion=Free
  swing2LimitDegrees=45
  swing2Motion=Free
  twistLimitDegrees=45
  twistMotion=Free
 }
 {
  bone1Name=calf_l
  bone2Name=thigh_l
  swing1LimitDegrees=45
  swing1Motion=Free
  swing2LimitDegrees=45
  swing2Motion=Free
  twistLimitDegrees=45
  twistMotion=Free
 }
 {
  bone1Name=thigh_r
  bone2Name=pelvis
  swing1LimitDegrees=45
  swing1Motion=Free
  swing2LimitDegrees=45
  swing2Motion=Free
  twistLimitDegrees=45
  twistMotion=Free
 }
 {
  bone1Name=calf_r
  bone2Name=thigh_r
  swing1LimitDegrees=45
  swing1Motion=Free
  swing2LimitDegrees=45
  swing2Motion=Free
  twistLimitDegrees=45
  twistMotion=Free
 }
 {bone1Name=dyn_wing_02_l, bone2Name=Wing_01_L, swing1LimitDegrees=17, swing2LimitDegrees=3, twistLimitDegrees=45, twistMotion=Locked}
 {bone1Name=dyn_wing_03_l, bone2Name=dyn_wing_02_l, swing1LimitDegrees=23, swing2LimitDegrees=5, twistLimitDegrees=45, twistMotion=Locked}
 {bone1Name=dyn_wing_04_l, bone2Name=dyn_wing_03_l, swing1LimitDegrees=80, swing2LimitDegrees=7, twistLimitDegrees=45}
 {bone1Name=dyn_wing_05_l, bone2Name=dyn_wing_04_l, swing1LimitDegrees=80, swing2LimitDegrees=9, twistLimitDegrees=45, twistMotion=Locked}
 {bone1Name=dyn_wing_07_l, bone2Name=dyn_wing_06_l, swing1LimitDegrees=80, swing2LimitDegrees=12, twistLimitDegrees=45, twistMotion=Locked}
 {bone1Name=dyn_wing_08_l, bone2Name=dyn_wing_07_l, swing1LimitDegrees=60, swing2LimitDegrees=14, twistLimitDegrees=45, twistMotion=Locked}
 {bone1Name=dyn_wing_02_r, bone2Name=Wing_01_R, swing1LimitDegrees=25, swing2LimitDegrees=0, twistLimitDegrees=45, twistMotion=Locked}
 {bone1Name=dyn_wing_03_r, bone2Name=dyn_wing_02_r, swing1LimitDegrees=30, swing2LimitDegrees=0.1, twistLimitDegrees=45, twistMotion=Locked}
 {bone1Name=dyn_wing_04_r, bone2Name=dyn_wing_03_r, swing1LimitDegrees=80, swing2LimitDegrees=0.1, twistLimitDegrees=45}
 {bone1Name=dyn_wing_05_r, bone2Name=dyn_wing_04_r, swing1LimitDegrees=80, swing2LimitDegrees=0.4, twistLimitDegrees=45, twistMotion=Locked}
 {bone1Name=dyn_wing_06_r, bone2Name=dyn_wing_05_r, swing1LimitDegrees=80, swing2LimitDegrees=0.4, twistLimitDegrees=45, twistMotion=Locked}
 {bone1Name=dyn_wing_07_r, bone2Name=dyn_wing_06_r, swing1LimitDegrees=80, swing2LimitDegrees=0.5, twistLimitDegrees=45, twistMotion=Locked}
 {bone1Name=dyn_wing_08_r, bone2Name=dyn_wing_07_r, swing1LimitDegrees=60, swing2LimitDegrees=0.5, twistLimitDegrees=45, twistMotion=Locked}
 {bone1Name=dyn_wing_06_l, bone2Name=dyn_wing_05_l, swing1LimitDegrees=80, swing2LimitDegrees=10, twistLimitDegrees=45, twistMotion=Locked}
]
